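; Ribbon manager for the Unstable Liquid Gun that detonates the liquid trail as a chain of explosions.
; The server starts a fuse immediately or on first projectile landing, backed by a failsafe, then multicasts detonation.
; Timed steps batch trail points by distance and count, each playing effects and applying server-side radial, line-of-sight GameplayEffect damage.
; Damage scales with batched points and vehicle or environment multipliers, while high blinking intensity can widen the radius.
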(event Custom|StartExplodingFailsafe
  ;[0.4143 1 0.5232] Begin Play and Fuse Handling
  (Development|PrintString "StartExplodingFailsafe" (Variables|Default|GetbDebug) (Variables|Default|GetbDebug))
  (CallFunction|StartExploding))

(event EventBeginPlay
  ;[0.4143 1 0.5232] Begin Play and Fuse Handling
  (|CallParentFunction)
  (Networking|SwitchHasAuthority
    (:Authority
      (if (ScalableFloat|GetValueasBool (Variables|Default|GetWaitToStartFuseUntilFirstProjectileTouchesGround))
        (bind _returnvalue (Utilities|Time|SetTimerbyEvent (delegate Custom|StartExplodingFailsafe) (ScalableFloat|GetValueatLevel (Variables|Default|GetTimeBeforeFailsafeExplosionStarts)))) ; If waiting for the first projectile to hit before starting the fuse, also start a failsafe timer.
        (Variables|Default|SetStartExplodingFailsafeTimer _returnvalue)
        (else
          (CallFunction|StartFuse))))))

(event EventFirstProjectileStopped(serveronly)
  ;[0.4143 1 0.5232] Begin Play and Fuse Handling
  (|CallParentFunction)
  (Networking|SwitchHasAuthority
    (:Authority
      (if (ScalableFloat|GetValueasBool (Variables|Default|GetWaitToStartFuseUntilFirstProjectileTouchesGround))
        (CallFunction|StartFuse)))))

(event Custom|StartFuse
  ;[0.4143 1 0.5232] Begin Play and Fuse Handling
  (Development|PrintString "StartFuse" (Variables|Default|GetbDebug) (Variables|Default|GetbDebug))
  (Utilities|Time|SetTimerbyEvent (delegate Custom|StartExploding) (ScalableFloat|GetValueatLevel (Variables|Default|GetTimeBeforeExplosionStarts_))))

(event Custom|StartExploding
  ;[0.4143 1 0.5232] Begin Play and Fuse Handling
  (Utilities|Time|ClearandInvalidateTimerbyHandle (Variables|Default|GetStartExplodingFailsafeTimer))
  (CallFunction|StartExplodingMulticast))

(event Custom|StartExploding_Multicast
  ;[0.4143 1 0.5232] Begin Play and Fuse Handling
  (Development|PrintString "Start Exploding" (Variables|Default|GetbDebug) (Variables|Default|GetbDebug))
  (Variables|Default|SetTimePerPoint (* (Math|Float|SafeDivide (Math|Float|SafeDivide 1.0 (Variables|Default|GetFireRate)) (Math|Conversions|ToFloat(Integer) (Variables|LiquidRibbonManager|GetPointsPerSegment))) (ScalableFloat|GetValueatLevel (Variables|Default|GetIndividualExplosionDelayMultiplier))))
  (CallFunction|SetUpNextExplosion))

(event EventEndPlay
  (params (EndPlayReason :type EEndPlayReason%20Enum :default "Destroyed"))
  ;[1 0.6119 0.3488] EndPlay - Cleanup
  (|CallParentFunction EndPlayReason)
  (Utilities|Time|ClearandInvalidateTimerbyHandle (Variables|Default|GetStartExplodingFailsafeTimer)))

(event Custom|SetUpNextExplosion
  (bind _pointstoexplode (Variables|Default|GetPointsToExplode))
  (bind _currentexplosionposition (Variables|Default|GetCurrentExplosionPosition))
  (bind _returnvalue (+ _currentexplosionposition _pointstoexplode))
  (bind _allpositions (Variables|Default|GetAllPositions))
  (bind _output (Utilities|Array|Get(acopy) _allpositions _returnvalue))
  (if (not (Variables|Default|GetDoneExploding))
    (Variables|Default|SetDistanceSinceExploded 0.0)
    (Variables|Default|SetPointsToExplode 0)
    (Variables|Default|SetPreviousPosition (Utilities|Array|Get(acopy) (Variables|Default|GetAllPositions) (Variables|Default|GetCurrentExplosionPosition)))
    (Variables|Default|SetIn%20String "")
    (Utilities|FlowControl|ForLoopwithBreak 0 (ScalableFloat|GetValueasInteger (Variables|Default|GetMaxPointsPerExplosion_))
      (:LoopBody
        (Variables|Default|SetDistanceSinceExploded (+ (Math|Vector|Distance(Vector) _output (Variables|Default|GetPreviousPosition)) (Variables|Default|GetDistanceSinceExploded)))
        (Variables|Default|SetPreviousPosition _output)
        (Math|Integer|IncrementInt (Variables|Default|GetPointsToExplode)
          (:%20%20
            (if (>= (+ (Variables|Default|GetCurrentExplosionPosition) (Variables|Default|GetPointsToExplode)) (- (Utilities|Array|Length (Variables|Default|GetAllPositions)) 1))
              _)
            (if (or (and (> (Variables|Default|GetDistanceSinceExploded) (ScalableFloat|GetValueatLevel (Variables|Default|GetMinDistanceForExplosion))) (> (Variables|Default|GetPointsToExplode) (ScalableFloat|GetValueatLevel (Variables|Default|GetMinPointsPerExplosion_)))) (>= (Variables|Default|GetPointsToExplode) (ScalableFloat|GetValueatLevel (Variables|Default|GetMaxPointsPerExplosion_))))
              _))))
      (:Completed
        _))
    (bind _output_get (Variables|Default|SetNextExplosionDelay (* (Variables|Default|GetPointsToExplode) (Variables|Default|GetTimePerPoint))))
    (Utilities|Time|SetTimerbyEvent (delegate Custom|TriggerNextExplosion) _output_get)))

(event Custom|AlllowAnotherAudioVisualizer
  (Utilities|FlowControl|DoOnce false
    (:Completed
      (Utilities|Time|SetTimerbyEvent (delegate Custom|AlllowAnotherAudioVisualizer) 0.5)
      (Fortnite|ShowSoundIndicatorLocal :Location (Utilities|GetCurrentLocation) :MaxAudibleDistance (Variables|/Script/Engine.SoundBase|GetMaxDistance (Variables|Default|GetExplodingSound)) :IndicatorType "Gunshot" :Instigator self :Tint "(R=0.000000,G=0.000000,B=0.000000,A=0.000000)" :AttenuationMaxRangeMultiplier 1.0 :AttenuationIntensityMultiplier 1.0))))

(event Custom|TriggerNextExplosion
  (bind _output (Utilities|GetCurrentLocation))
  (bind _ns_unstableliquidgun_explosion_loop_0 (Variables|Default|GetNS_UnstableLiquidGun_Explosion_Loop_0))
  (bind _ns_unstableliquidgun_explosion_loop_0_2 (Variables|Default|GetNS_UnstableLiquidGun_Explosion_Loop_0))
  (Variables|Default|SetCurrentExplosionPosition (+ (Variables|Default|GetCurrentExplosionPosition) (Variables|Default|GetPointsToExplode)))
  (Utilities|FlowControl|Sequence
    (:then_0
      (CallFunction|Damage (> (Utilities|Array|Get(acopy) (Variables|Default|GetBlinkingIntensityArray) (Variables|Default|GetCurrentExplosionPosition)) 0.5))
      (Utilities|FlowControl|Sequence
        (:then_0
          (Networking|CanExecuteCosmeticEvents
            (:True
              (Utilities|FlowControl|Sequence
                (:then_0
                  (Components|Activation|Activate _ns_unstableliquidgun_explosion_loop_0 false) ; Triggering a Looping FX would be way cheaper
                  (Transformation|SetWorldLocation _ns_unstableliquidgun_explosion_loop_0 _output false true)
                  (Transformation|SetWorldRotation _ns_unstableliquidgun_explosion_loop_0 (Math|Rotator|MakeRotfromZ (CallFunction|GetNormalatPosition (Variables|Default|GetCurrentExplosionPosition))) false false)
                  (ForceFeedback|SpawnForceFeedbackatLocation :ForceFeedbackEffect (Variables|Default|GetImpactForceFeedback) :Location _output :bLooping false :IntensityMultiplier 1.0 :StartTime 0.0 :AttenuationSettings "/UnstableLiquidGun/Items/UnstableLiquidGun/UnstableLiquidGun_FF_Attenuation.UnstableLiquidGun_FF_Attenuation" :bAutoDestroy true)
                  (bind _returnvalue (Audio|SpawnSoundatLocation (Variables|Default|GetExplodingSound) _output))
                  (bind _hitwater (Utilities|CheckforWater))
                  (Audio|Parameter|SetIntegerParameter _returnvalue "InWater" (Math|Conversions|ToInteger(Boolean) _hitwater) :!class /Script/AudioExtensions.AudioParameterControllerInterface)
                  (Utilities|FlowControl|DoOnce false
                    (:Completed
                      (Utilities|Time|SetTimerbyEvent (delegate Custom|AlllowAnotherAudioVisualizer) 0.5)
                      (Fortnite|ShowSoundIndicatorLocal :Location _output :MaxAudibleDistance (Variables|/Script/Engine.SoundBase|GetMaxDistance (Variables|Default|GetExplodingSound)) :IndicatorType "Gunshot" :Instigator self :Tint "(R=0.000000,G=0.000000,B=0.000000,A=0.000000)" :AttenuationMaxRangeMultiplier 1.0 :AttenuationIntensityMultiplier 1.0))))
                (:then_1
                  (Effects|Components|ParticleSystem|SetBoolParameter _ns_unstableliquidgun_explosion_loop_0_2 "bIsIntensiveExplosion" (> (Utilities|Array|Get(acopy) (Variables|Default|GetBlinkingIntensityArray) (Variables|Default|GetCurrentExplosionPosition)) 0.5))
                  (bind _hitwater_2 (Utilities|CheckforWater))
                  (Effects|Components|ParticleSystem|SetBoolParameter _ns_unstableliquidgun_explosion_loop_0_2 "bSpawnWater" _hitwater_2))))))
        (:then_1
          (if (>= (Variables|Default|GetCurrentExplosionPosition) (- (Utilities|Array|Length (Variables|Default|GetAllPositions)) 1))
            (Variables|Default|SetDoneExploding true)
            (Variables|Default|SetCurrentExplosionPosition (Utilities|Array|Length (Variables|Default|GetAllPositions)))
            (CallFunction|CEStopExplosion)
            (CallFunction|CleanUp)))
        (:then_2
          (Utilities|FlowControl|RetriggerableDelay 0.2) ; Is there a better place to deactive the looping explosion FX?
          (Components|Activation|Deactivate (Variables|Default|GetNS_UnstableLiquidGun_Explosion_Loop_0)))))
    (:then_1
      (CallFunction|SetUpNextExplosion))))

(event Custom|Damage
  (params (isIntenseVersion :type Boolean))
  (CallFunction|DamageFunc isIntenseVersion))

(fn ApplyDamage
  (params (Array :type Array%20of%20Actor%20Object%20References))
  (bind _minpointsperexplosion (Variables|Default|GetMinPointsPerExplosion_))
  (bind _returnvalue (ScalableFloat|GetValueatLevel _minpointsperexplosion))
  (bind _mindamageperexplosion (Variables|Default|GetMinDamagePerExplosion))
  (bind _returnvalue_2 (ScalableFloat|GetValueatLevel _mindamageperexplosion))
  (bind _maxdamageperexplosion (Variables|Default|GetMaxDamagePerExplosion))
  (bind _returnvalue_3 (ScalableFloat|GetValueatLevel _maxdamageperexplosion))
  (bind _maxpointsperexplosion (Variables|Default|GetMaxPointsPerExplosion_))
  (bind _returnvalue_4 (ScalableFloat|GetValueatLevel _maxpointsperexplosion))
  (bind _instigator (Variables|Actor|GetInstigator))
  (bind _returnvalue_5 (Ability|GetFortAbilitySystemComponent _instigator))
  (bind _pointstoexplode (Variables|Default|GetPointsToExplode))
  (bind _returnvalue_6 (Math|Conversions|ToFloat(Integer) _pointstoexplode))
  (bind _output (Utilities|RemapFloat _returnvalue_6 (Math|Vector2D|MakeVector2D _returnvalue _returnvalue_4) (Math|Vector2D|MakeVector2D _returnvalue_2 _returnvalue_3) true))
  (for _array_element Array
    (Utilities|CheckLOS _array_element
      (:Not%20Blocked
        (Utilities|Array|Add (LocalVariables|GetActorsToDamage) _array_element))))
  (for _array_element_2 (LocalVariables|GetActorsToDamage)
    (bind _returnvalue_7 (Ability|GameplayEffect|AssignTagSetbyCallerMagnitude (GameplayEffects|MakeOutgoingSpec _returnvalue_5 (Variables|Default|GetExplosionGE) 0.0 (Class|AthenaFunctionLibrary|MakeSimpleFortEffectContext (Variables|/Script/Engine.Actor|GetInstigator (Variables|Default|GetWeapon)) (Variables|Default|GetWeapon))) "(TagName=\"SetByCaller.Damage\")" (* _output (Math|Float|SelectFloat (ScalableFloat|GetValueatLevel (Variables|Default|GetVehicleDamageMultiplier)) 1.0 (.bSuccess (Utilities|Casting|CastToFortVehicleInterface _array_element_2))))))
    (bind _returnvalue_8 (Ability|GameplayEffect|AssignTagSetbyCallerMagnitude _returnvalue_7 "(TagName=\"SetByCaller.EnvironementalDamage\")" (* _output (ScalableFloat|GetValueatLevel (Variables|Default|GetEnvironmentDamageMultiplier_)))))
    (GameplayEffects|ApplyGameplayEffectSpecToTarget _returnvalue_5 _returnvalue_8 (Ability|GetFortAbilitySystemComponent _array_element_2))
    (Development|PrintString (Utilities|String|Append "Damaging " (Utilities|GetDisplayName _array_element_2)) (Variables|Default|GetbDebug) (Variables|Default|GetbDebug))))

(fn BP_GetFirstVisibleIndex
  (return (Variables|Default|GetCurrentExplosionPosition)))

(fn DamageFunc
  (params (isIntenseVersion :type Boolean))
  (bind _explosionradius (Variables|Default|GetExplosionRadius_))
  (bind _returnvalue (ScalableFloat|GetValueatLevel _explosionradius))
  (bind _explosionradiusmultiplierwhenintense (Variables|Default|GetExplosionRadiusMultiplierWhenIntense))
  (bind _returnvalue_2 (ScalableFloat|GetValueatLevel _explosionradiusmultiplierwhenintense))
  (bind _returnvalue_3 (Math|Float|SelectFloat _returnvalue_2 1.0 isIntenseVersion))
  (bind _currentexplosionposition (Variables|Default|GetCurrentExplosionPosition))
  (bind _allpositions (Variables|Default|GetAllPositions))
  (bind _output (Utilities|Array|Get(acopy) _allpositions _currentexplosionposition))
  (bind _returnvalue_4 (* _returnvalue _returnvalue_3))
  (Networking|SwitchHasAuthority
    (:Authority
      (if (Variables|Default|GetbDebug)
        (Rendering|Debug|DrawDebugSphere _output _returnvalue_4 12 "(R=1.000000,G=0.093796,B=0.000000,A=1.000000)" 5.0 1e+01)
        (else
          _))
      (bind _outcomponents (Collision|SphereOverlapComponents _output _returnvalue_4 (Utilities|Array|MakeArray "ObjectTypeQuery6" "ObjectTypeQuery1" "ObjectTypeQuery5" "ObjectTypeQuery2" "ObjectTypeQuery4" "ObjectTypeQuery3" "ObjectTypeQuery13" "ObjectTypeQuery7") 0 (Utilities|Array|MakeArray self)))
      ;[0.4354 1 0.506] Filter out components that ignore FCC_TraceWeaponCapsule
      (for _array_element _outcomponents
        (switch Utilities|FlowControl|Switch|SwitchonECollisionResponse (Collision|GetCollisionResponsetoChannel _array_element "ECC_GameTraceChannel3")
          (:ECR_Overlap
            _)
          (:ECR_Block
            _))
        (Utilities|Array|Add (LocalVariables|GetHitComponents) _array_element))
      (bind _outactorlist (Actor|GetActorListfromComponentList (LocalVariables|GetHitComponents)))
      (CallFunction|ApplyDamage _outactorlist))))

(fn GetAffiliationsToShowFor
  (return (Utilities|Array|MakeArray "Friendly" "Hostile")))

(fn GetNormalAtPosition
  (params (Position :type Integer))
  (bind _pointspersegment (Variables|LiquidRibbonManager|GetPointsPerSegment))
  (bind _position (Variables|GetPosition))
  (bind _returnvalue (/ _position _pointspersegment))
  (bind _output (Utilities|Array|Get(acopy) (Variables|Default|GetProjectiles) _returnvalue))
  (bind _projectiles (Variables|Default|GetProjectiles))
  (Utilities|FlowControl|Sequence
    (:then_0
      (if (> _pointspersegment 0)
        (if (Utilities|Array|IsValidIndex _projectiles _returnvalue)
          (Utilities|IsValid _output
            (:Is%20Valid
              (return (* (Variables|/UnstableLiquidGun/Items/UnstableLiquidGun/B_Prj_Ranged_NitroglycerineGun.B_Prj_Ranged_NitroglycerineGun_C|GetDirectionPrjHitFrom _output) -1.0)))))))
    (:then_1
      (return "0, 0, 0"))))

(fn ShouldShowSoundIndicator
  (params (PlayerController :type Fort%20Player%20Controller%20Object%20Reference))
  (return true))

(fn ConstructionScript
  (|CallParentFunction))
