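; Projectile for the Nitroglycerine (goo) gun, forming one point in a liquid chain drawn by a ribbon manager.
; Gravity increases after a delay; hits and validated pawn overlaps multicast a stop, attaching to vehicles or pawns and resuming if the base is destroyed.
; On stop, it plays landing effects, nudges away from close neighbors and flags breaks from accumulated curvature to prevent drawing.
; The timer-based capsule collision check is deprecated.
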(event EventBeginPlay
  ;[1 1 1] BeginPlay
  (Utilities|FlowControl|Sequence
    (:then_0
      (Variables|BreakageVariables|SetRandomVectorOne (Math|Random|RandomUnitVector))
      (Variables|BreakageVariables|SetRandomVectorTwo (Math|Random|RandomUnitVector)))
    (:then_1
      (Variables|Default|SetUpdatePosition (Transformation|GetActorLocation))
      (Projectile|SetGravityScale (ScalableFloat|GetValueatLevel (Variables|Default|GetInitialGravity)))
      (Utilities|Time|SetTimerbyEvent (delegate Custom|IncreaseGravity) (ScalableFloat|GetValueatLevel (Variables|Default|GetDelayBeforeIncreasedGravity))))
    (:then_2)
    (:then_3)))

(event OnComponentEndOverlap(CapsuleComponent)
  (params (OverlappedComponent :type Primitive%20Component%20Object%20Reference) (OtherActor :type Actor%20Object%20Reference) (OtherComp :type Primitive%20Component%20Object%20Reference) (OtherBodyIndex :type Integer)))

(event Custom|IncreaseGravity
  ;[1 1 1] BeginPlay
  (Projectile|SetGravityScale (ScalableFloat|GetValueatLevel (Variables|Default|GetIncreasedGravity))))

(event Custom|StartCheckingCollision
  ;[1 0 0] Deprecated Capsule Trace Collision
  (Utilities|FlowControl|DelayUntilNextTick)
  (bind _returnvalue (Utilities|Time|SetTimerbyEvent (delegate Custom|CheckCollision) 0.05 true))
  (Variables|Default|SetCheckCollisionTimer _returnvalue)
  (Variables|Default|SetCheckingCollision true)
  (CallFunction|GetLiquidPoints)
  (CallFunction|UpdateLastCheckedPositions false))

(event Custom|UpdateLastCheckedPositions
  (params (IsFirstUpdate :type Boolean))
  ;[1 0 0] Deprecated Capsule Trace Collision
  (if (not IsFirstUpdate)
    (Variables|Default|SetLastMovement (- (Variables|Default|GetLiquidCenterPoint) (Variables|Default|GetLastCheckedCenterPoint))))
  (Variables|Default|SetLastCheckedCenterPoint (Variables|Default|GetLiquidCenterPoint))
  (Variables|Default|SetLastCheckedStartPoint (Variables|Default|GetLiquidStartPoint))
  (Variables|Default|SetLastCheckedEndPoint (Variables|Default|GetLiquidEndPoint)))

(event Custom|Deprecated%20Update%20collider)

(event Custom|CheckCollision
  ;[1 0 0] Deprecated Capsule Trace Collision
  (Utilities|DoOncePerFrame
    (:%20%20
      (CallFunction|GetLiquidPoints)
      (Utilities|FlowControl|Sequence
        (:then_1
          (Networking|SwitchHasAuthority
            (:Authority
              (if (> (Class|AbilityFunctions|GetAngleBetweenTwoVectors (- (Variables|Default|GetLiquidEndPoint) (Variables|Default|GetLiquidCenterPoint)) (- (Variables|Default|GetLiquidCenterPoint) (Variables|Default|GetLiquidStartPoint))) 3e+01)
                (Utilities|FlowControl|Sequence
                  (:then_0
                    (if (not (Utilities|IsLastProjectile))
                      (bind (_outhit _returnvalue) (CallFunction|DeprecatedDoCollisionCheck (Variables|Default|GetLastCheckedEndPoint) (Variables|Default|GetLastCheckedCenterPoint) (Variables|Default|GetLiquidEndPoint) (Variables|Default|GetLiquidCenterPoint)))
                      (if _returnvalue
                        (CallFunction|HitMulticast _outhit))))
                  (:then_1
                    (if (not (Utilities|IsFirstProjectile))
                      (bind (_outhit_2 _returnvalue_2) (CallFunction|DeprecatedDoCollisionCheck (Variables|Default|GetLastCheckedStartPoint) (Variables|Default|GetLastCheckedCenterPoint) (Variables|Default|GetLiquidStartPoint) (Variables|Default|GetLiquidCenterPoint)))
                      (if _returnvalue_2
                        (CallFunction|HitMulticast _outhit_2)))))
                (else
                  (bind (_outhit_3 _returnvalue_3) (CallFunction|DeprecatedDoCollisionCheck (Variables|Default|GetLastCheckedStartPoint) (Variables|Default|GetLastCheckedEndPoint) (Variables|Default|GetLiquidStartPoint) (Variables|Default|GetLiquidEndPoint)))
                  (if _returnvalue_3
                    (CallFunction|HitMulticast _outhit_3))))
              (Variables|Default|SetCheckingCollision false)
              (Utilities|Time|ClearandInvalidateTimerbyHandle (Variables|Default|GetCheckCollisionTimer)))))
        (:then_2
          (CallFunction|UpdateLastCheckedPositions false))))))

(event Custom|CustomEvent_0
  (params (WasCancelled :type Boolean)))

(event Custom|HitActor
  (params (HitActor :type Actor%20Object%20Reference))
  ;[1 0 0] Deprecated
  (Utilities|Casting|CastToB_Prj_Ranged_NitroglycerineGun HitActor
    (:CastFailed
      (Variables|Default|SetLastOverlapTime (Utilities|Time|GetGameTimeinSeconds))
      (if (Variables|Default|GetCheckingCollision)
        _))))

(event EventOnStop
  (params (Hit :type Hit%20Result%20Structure%20%28by%20ref%29))
  (bind (_bblockinghit _binitialoverlap _time _distance _location _impactpoint _normal _impactnormal _physmat) (Collision|BreakHitResult Hit))
  ;[1 1 1] StopBehavior, get PhysMat
  (Utilities|IsValid _physmat
    (:Is%20Valid
      (Variables|Effects|SetStopSurfaceType (Variables|/Script/PhysicsCore.PhysicalMaterial|GetSurfaceType _physmat)))
    (:Is%20Not%20Valid
      _))
  (Variables|Default|SetStopLocation _impactpoint) ; Use Loc and Rot for single explosion
  (Variables|Default|SetStopRotX (Math|Rotator|MakeRotfromX _impactnormal))
  (Variables|FortLiquidGunProjectile|SetStopNormal _impactnormal)
  (Variables|Default|SetStopRotZ (Math|Rotator|MakeRotfromZ _impactnormal))
  (Variables|Default|SetHasStopped true)
  (Utilities|FlowControl|Sequence
    (:then_0
      (Networking|CanExecuteCosmeticEvents
        (:True
          (if (Utilities|Enum|Equal(Enum) (Variables|Effects|GetStopSurfaceType) "SurfaceType8")
            (Networking|CanExecuteCosmeticEvents
              (:True
                (bind _returnvalue (Effects|SpawnEmitteratLocation (Variables|Effects|GetWaterLandDelayPS) (Variables|Default|GetStopLocation) (Variables|Default|GetStopRotX)))
                (Variables|Default|SetWaterLandDelayRipplePSC _returnvalue)
                (Audio|PlaySoundatLocation (Variables|Audio|GetSFX_Land_Water) (Variables|Default|GetStopLocation))))
            (else
              (Audio|PlaySoundatLocation (Variables|Audio|GetSFX_Land) (Variables|Default|GetStopLocation)))))))
    (:then_1
      (Utilities|FlowControl|DelayUntilNextTick)
      (CallFunction|CheckforTooClosetoNeighbors)
      ;;[1 0.6755 0.4816] Determine if we need a detachment (to break) based on our curvature. This helps prevent writing/drawing with the gun
      ;; Note: this does not actually make the detachment, just marks this projectile's segment as needing one. The detachment is handled in the ribbon manager like normal
      (CallFunction|TryToBreak))))

(event Custom|SetPreviousPrj
  (params (PreviousPrj :type Actor%20Object%20Reference))
  (Utilities|IsValid PreviousPrj
    (:Is%20Valid
      (if (< (Actor|GetGameTimeSinceCreation PreviousPrj) 0.25)
        (Variables|Default|SetPreviousPrj PreviousPrj)))))

(event Custom|SetNextPrj
  (params (NextPrj :type B%20Prj%20Ranged%20Nitroglycerine%20Gun%20Object%20Reference))
  (if (< (Actor|GetGameTimeSinceCreation) 0.25)
    (Variables|Default|SetNextPrj NextPrj)))

(event Collision|EventHit
  (params (MyComp :type Primitive%20Component%20Object%20Reference) (Other :type Actor%20Object%20Reference) (OtherComp :type Primitive%20Component%20Object%20Reference) (bSelfMoved :type Boolean) (HitLocation :type Vector :default "0, 0, 0") (HitNormal :type Vector :default "0, 0, 0") (NormalImpulse :type Vector :default "0, 0, 0") (Hit :type Hit%20Result%20Structure%20%28by%20ref%29))
  ;[0.2228 0.3231 1] Direct Hit Handling
  ; Collision test on the spline is handled by CheckAndHandleSpline collision which is called from the ribbon manager
  (CallFunction|HitMulticast Hit) ; needed because hit sometimes only occurs on server for some reason
)

(event Custom|MoveProjectile%20%28Multicast%29
  (params (NewLocation :type Vector))
  (bind _location_i_m_moving_to (Variables|SpreadingOut|GetLocation%20I'm%20Moving%20To))
  (bind _location_i_m_moving_from (Variables|SpreadingOut|GetLocation%20I'm%20Moving%20From))
  (if (not (Math|Vector|VectorIsNAN NewLocation))
    (if (ScalableFloat|GetValueasBool (Variables|Default|GetSF_InterpolateWhenMovingAwayFromNeighbors))
      (Variables|SpreadingOut|SetLocation%20I'm%20Moving%20To NewLocation)
      (Variables|SpreadingOut|SetLocation%20I'm%20Moving%20From (Transformation|GetActorLocation))
      (bind (_direction _progress) (Timeline|MoveProjectile PlayFromStart)
        (:Update
          (if (not (Math|Vector|VectorIsNAN (Math|Vector|Lerp(Vector) _location_i_m_moving_from _location_i_m_moving_to _progress)))
            (Transformation|SetActorLocation (Math|Vector|Lerp(Vector) _location_i_m_moving_from _location_i_m_moving_to _progress))
            (else
              (Fortnite|FortShippingLog 0 "B_Prj_Ranged_NitroglycerineGun::MoveProjectile - Detected NaN Lerped Location" true)))))
      (else
        (Transformation|SetActorLocation NewLocation false true)))
    (else
      (Fortnite|FortShippingLog 0 "B_Prj_Ranged_NitroglycerineGun::MoveProjectile - Detected NaN New Location" true))))

(event Custom|HitMulticast
  (params (LastHit :type Hit%20Result%20Structure))
  (bind _capsulecomponent (Variables|Collision|GetCapsuleComponent))
  (bind _returnvalue (Components|Capsule|GetScaledCapsuleRadius _capsulecomponent))
  (bind _lasthit (Variables|Default|GetLastHit))
  (bind (_bblockinghit _binitialoverlap _time _distance _location _impactpoint _normal _impactnormal _physmat _hitactor _hitcomponent _hitbonename _bonename _hititem _elementindex _faceindex _tracestart) (Collision|BreakHitResult _lasthit))
  (bind _lasthit_2 (Variables|Default|GetLastHit))
  (bind (_bblockinghit_2 _binitialoverlap_2 _time_2 _distance_2 _location_2 _impactpoint_2 _normal_2 _impactnormal_2 _physmat_2 _hitactor_2) (Collision|BreakHitResult _lasthit_2))
  (bind (_asfort_vehicle_interface _bsuccess) (Utilities|Casting|CastToFortVehicleInterface _hitactor_2))
  (bind (_asfort_vehicle_interface_2 _bsuccess_2) (Utilities|Casting|CastToFortVehicleInterface _hitactor_2))
  (bind _directionprjhitfrom (Variables|Default|GetDirectionPrjHitFrom))
  (bind _returnvalue_2 (Utilities|Operators|Multiply _directionprjhitfrom _returnvalue -1.0))
  (bind _returnvalue_3 (+ _impactpoint _returnvalue_2))
  ;[0.305 1 0.7875] Hit Handling
  (if (not (Math|Vector|VectorIsNAN _impactpoint))
    (Variables|Default|SetLastHit LastHit)
    (Variables|Default|SetDirectionPrjHitFrom (Math|Vector|Normalize (- _impactpoint (Math|Vector|SelectVector (Transformation|GetActorLocation) _tracestart (Math|Vector|VectorIsNAN _tracestart)))))
    (Utilities|IsValid (Variables|Collision|GetCapsuleComponent)
      (:Is%20Valid
        (if (not (Math|Vector|VectorIsNAN _returnvalue_3))
          (Default|CallOnStopped)
          (Game|Components|ProjectileMovement|StopSimulating (Variables|Movement|GetProjectileMovementComponent) (Variables|Default|GetLastHit))
          (Collision|SetActorEnableCollision false)
          (Variables|Default|SetProject true)
          (bind _output_get (Variables|Default|SetImpact%20Point _returnvalue_3))
          (Transformation|SetActorLocation _output_get false true)
          (Variables|Default|SetHitNormal _normal_2)
          (Utilities|IsValid _hitactor_2
            (:Is%20Valid
              (Utilities|FlowControl|Sequence
                (:then_0
                  (if (Math|Boolean|ANDBoolean (not (Variables|Default|GetAttachedToPawn)) (not (Variables|Default|GetAttachedToVehicle)) (not (Variables|Default|GetHasTriedToAttach)))
                    (Utilities|FlowControl|Sequence
                      (:then_0
                        (if (and (ScalableFloat|GetValueasBool (Variables|Default|GetAttachToVehicles)) _bsuccess_2)
                          (Transformation|AttachActorToActor self _hitactor_2 "None" "KeepWorld" "KeepWorld" "KeepWorld" true)
                          (Variables|Default|SetAttachedToVehicle true)
                          (Collision|SetActorEnableCollision false)))
                      (:then_1
                        (if (Math|Boolean|ANDBoolean (!= _hitactor_2 (Variables|Actor|GetInstigator)) (ScalableFloat|GetValueasBool (Variables|Default|GetAttachToPlayers)) (not _bsuccess))
                          (bind _aspawn (Utilities|Casting|CastToPawn _hitactor_2))
                          (CallFunction|AttachtoPawnMulticast _aspawn)))
                      (:then_2
                        (Variables|Default|SetHasTriedToAttach true)))))
                (:then_1
                  (Game|BindEventtoOnDestroyed _hitactor_2 (delegate Custom|OnAttachedWallDestroyed))
                  (bind _asbuilding_actor (Utilities|Casting|CastToBuildingActor _hitactor_2))
                  (EventDispatchers|BindEventtoOnDied _asbuilding_actor (delegate Custom|BaseDied))))))
          (else
            (Fortnite|FortShippingLog 0 "B_Prj_Ranged_NitroglycerineGun::HitMulticast - Detected NaN adjusted ImpactPoint" true)))))
    (else
      (Fortnite|FortShippingLog 0 "B_Prj_Ranged_NitroglycerineGun::HitMulticast - Detected NaN ImpactPoint" true))))

(event Custom|OnAttachedWallDestroyed
  (params (DestroyedActor :type Actor%20Object%20Reference))
  ;[0.305 1 0.7875] Hit Handling
  (Utilities|IsValid (Variables|Movement|GetProjectileMovementComponent)
    (:Is%20Valid
      (Projectile|ResumeSimulation)
      (Components|Movement|SetUpdatedComponent (Variables|Movement|GetProjectileMovementComponent) (Variables|Collision|GetCapsuleComponent))
      (Variables|/Script/Engine.ProjectileMovementComponent|SetbSimulationEnabled true (Variables|Movement|GetProjectileMovementComponent))
      (Collision|SetActorEnableCollision true))))

(event Custom|BaseDied
  (params (DamagedActor :type Actor%20Object%20Reference) (Damage :type Float%20%28single-precision%29) (InstigatedBy :type Controller%20Object%20Reference) (DamageCauser :type Actor%20Object%20Reference) (HitLocation :type Vector) (FHitComponent :type Primitive%20Component%20Object%20Reference) (BoneName :type Name) (Momentum :type Vector))
  ;[0.305 1 0.7875] Hit Handling
  (Utilities|IsValid (Variables|Movement|GetProjectileMovementComponent)
    (:Is%20Valid
      (Projectile|ResumeSimulation)
      (Components|Movement|SetUpdatedComponent (Variables|Movement|GetProjectileMovementComponent) (Variables|Collision|GetCapsuleComponent))
      (Variables|/Script/Engine.ProjectileMovementComponent|SetbSimulationEnabled true (Variables|Movement|GetProjectileMovementComponent))
      (Collision|SetActorEnableCollision true))))

(event OnComponentBeginOverlap(CapsuleComponent)
  (params (OverlappedComponent :type Primitive%20Component%20Object%20Reference) (OtherActor :type Actor%20Object%20Reference) (OtherComp :type Primitive%20Component%20Object%20Reference) (OtherBodyIndex :type Integer) (bFromSweep :type Boolean) (SweepResult :type Hit%20Result%20Structure%20%28by%20ref%29))
  ;[0.2228 0.3231 1] Direct Hit Handling
  ; Collision test on the spline is handled by CheckAndHandleSpline collision which is called from the ribbon manager
  (bind _bvalidhit (CallFunction|OverlapIsValid OtherActor SweepResult bFromSweep))
  (if _bvalidhit
    (if bFromSweep
      (CallFunction|HitMulticast SweepResult) ; needed because hit sometimes only occurs on server for some reason
      (else
        (CallFunction|HitMulticast (Collision|MakeHitResult false false 0.0 0.0 (Transformation|GetActorLocation) (Transformation|GetActorLocation) "0,0,1" "0,0,1" 0 OtherActor 0 "None" "None" 0 0 0 (Transformation|GetActorLocation))) ; needed because hit sometimes only occurs on server for some reason
      ))))

(event Custom|AttachToPawnMulticast
  (params (ParentActor :type Actor%20Object%20Reference))
  (bind _returnvalue (Actor|GetComponentbyClass ParentActor "/Script/Engine.SkeletalMeshComponent"))
  (bind (_returnvalue_x _returnvalue_y _returnvalue_z) (Math|Vector|BreakVector (- (Transformation|GetActorLocation) (Transformation|GetWorldLocation (Variables|AttachOnHit|GetAttachComponent)))))
  (bind (_returnvalue_x_2 _returnvalue_y_2) (Math|Vector|BreakVector (* (Math|Random|RandomUnitVector) 1e+01)))
  (bind _rootcomponent (Variables|/Script/Engine.Actor|GetRootComponent ParentActor))
  ;[0.5371 1 0.6857] Attach To Pawn
  ; Attach the projectile to the pawn, but also save off the AttachComponent and AttachOffset as that will be used for the spline visuals
  (Utilities|IsValid ParentActor
    (:Is%20Valid
      (Transformation|AttachActorToActor ParentActor "None" "KeepWorld" "KeepRelative" "KeepWorld")
      (Utilities|IsValid _returnvalue
        (:Is%20Valid
          (Utilities|IsValid (Transformation|GetAttachParent _returnvalue)
            (:Is%20Valid
              (Variables|AttachOnHit|SetAttachComponent _returnvalue) ; Use skel mesh comp
            )
            (:Is%20Not%20Valid
              (Utilities|IsValid _rootcomponent
                (:Is%20Valid
                  (Variables|AttachOnHit|SetAttachComponent _rootcomponent) ; Use root comp
                )))) ; For some pawns (eg, FortAIPawns), their mesh is detached \non the server and does not update position with the actor.  \nIn that case use the root component instead. \nThis can result in different AttachComponents on client/server, \nbut they are close enough to not matter.  (Using the mesh \ncomponent is necessary on clients for smoothing)
          (Variables|Default|SetAttachedToPawn true)
          (Collision|SetActorEnableCollision false)
          (Variables|AttachOnHit|SetAttachOffset (Utilities|Operators|Add (Math|Vector|MakeVector 0.0 0.0 _returnvalue_z) (Math|Vector|MakeVector _returnvalue_x_2 _returnvalue_y_2))))
        (:Is%20Not%20Valid
          _)))))

(fn B_Prj_Ranged_NitroglycerineGun_AutoGenFunc
  (return false))

(fn BrokeFromAngle
  (return (Utilities|Enum|Equal(Enum) (CallFunction|GetDetachmentReason) "ByAngle")))

(fn BrokeFromDistance
  (return (Utilities|Enum|Equal(Enum) (CallFunction|GetDetachmentReason) "ByDistance")))

(fn CheckForTooCloseToNeighbors
  (bind _previousprj (Variables|Default|GetPreviousPrj))
  (if (ScalableFloat|GetValueasBool (Variables|Default|GetSF_MoveAwayFromCloseNeighbors))
    (Utilities|IsValid _previousprj
      (:Is%20Valid
        (bind _asb_prj_ranged_nitroglycerine_gun_bb (Utilities|Casting|CastToB_Prj_Ranged_NitroglycerineGun_BB _previousprj))
        (if (< (Math|Vector|VectorLengthSquared (- (Variables|Default|GetImpact%20Point _asb_prj_ranged_nitroglycerine_gun_bb) (Variables|Default|GetImpact%20Point))) (Math|Float|Square 15.0))
          (Variables|Default|SetBlockAngleDetach true)
          (Networking|SwitchHasAuthority
            (:Authority
              (CallFunction|MoveProjectile(Multicast) (+ (Transformation|GetActorLocation) (* (Math|Vector|RotateVectorAroundAxis (Math|Vector|CrossProduct (Variables|FortLiquidGunProjectile|GetStopNormal) (Math|Vector|MakeVector (.z (Variables|FortLiquidGunProjectile|GetStopNormal)) (.x (Variables|FortLiquidGunProjectile|GetStopNormal)) (.y (Variables|FortLiquidGunProjectile|GetStopNormal)))) (Math|Conversions|ToFloat(Integer) (* (Variables|FortLiquidGunProjectile|GetProjectileIndex) 75)) (Variables|FortLiquidGunProjectile|GetStopNormal)) (Utilities|RemapFloat (Math|Conversions|ToFloat(Integer) (Variables|FortLiquidGunProjectile|GetProjectileIndex)) "(X=0.000000,Y=6.000000)" "(X=40.000000,Y=10.000000)" true)))))))))))

(fn DeprecatedDoCollisionCheck
  (params (CapsuleLastEdge :type Vector) (MyLastPosition :type Vector) (CapsuleCurrentEdge :type Vector) (MyCurrentPosition :type Vector))
  (LocalVariables|SetCurrentVecToEdge (- CapsuleCurrentEdge MyCurrentPosition))
  (bind _output_get (LocalVariables|SetLastVecToEdge (- CapsuleLastEdge MyLastPosition)))
  (bind _output_get_2 (LocalVariables|SetLengthToLastEdge (Math|Vector|VectorLength _output_get)))
  (LocalVariables|SetLastDirToEdge (/ _output_get _output_get_2)))

(fn GetLiquidPoints
  (bind _returnvalue (Transformation|GetActorLocation))
  (bind _cachedbpribbonmanager (Variables|Default|GetCachedBPRibbonManager))
  (Utilities|IsValid _cachedbpribbonmanager
    (:Is%20Valid
      (if (> (Utilities|Array|Length (Variables|/UnstableLiquidGun/Items/UnstableLiquidGun/BP_LiquidGun_RibbonManager.BP_LiquidGun_RibbonManager_C|GetAllPositions _cachedbpribbonmanager)) 0)
        (Utilities|FlowControl|Sequence
          (:then_0
            (Variables|Default|SetLiquidStartPoint (Utilities|FetchLiquidStartPoint)))
          (:then_1
            (Variables|Default|SetLiquidCenterPoint (Utilities|FetchLiquidCenterPoint)))
          (:then_2
            (Variables|Default|SetLiquidEndPoint (Utilities|FetchLiquidEndPoint))))
        (else
          _)))
    (:Is%20Not%20Valid
      _))
  (Variables|Default|SetLiquidCenterPoint _returnvalue)
  (Variables|Default|SetLiquidEndPoint _returnvalue)
  (Variables|Default|SetLiquidStartPoint _returnvalue))

(fn GetShouldBreakChain
  (return (Variables|Default|GetShould%20Break%20Chain)))

(fn GetVector
  (return (- (Transformation|GetActorLocation) (Transformation|GetActorLocation (Variables|Default|GetPreviousPrj)))))

(fn HandleSplineHit
  (params (Hit :type Hit%20Result%20Structure%20%28by%20ref%29))
  (CallFunction|HitMulticast Hit)
  (return))

(fn OverlapIsValid
  (params (OverlappedActor :type Actor%20Object%20Reference) (HitResult :type Hit%20Result%20Structure) (bFromSweep :type Boolean))
  (bind (_bblockinghit _binitialoverlap _time _distance _location) (Collision|BreakHitResult (Variables|GetHitResult)))
  (Utilities|FlowControl|Sequence
    (:then_0
      (bind _aspawn (Utilities|Casting|CastToPawn OverlappedActor)
        (:then
          (if (!= _aspawn (Variables|Actor|GetInstigator))
            ;[0.9887 0.5696 1] In sports cars, the players capsule can stick out of the roof/windshield and thus the be detected as a hit when it appears the goo should hit the vehicle instead. To handle this, do an additonal trace from the projectile to the pawn.
            ; Currently this issue has only been detected when a player is in a vehicle, so only do this additional trace in that case. If the issue is found in other cases, this check may need to be expanded or implementation adjusted.
            ;
            ; If a hit is detected, just ignore this overlap. The projectile system will pick up the hit shortly (via Event Hit)
            ;
            (bind _asfort_player_pawn_athena (Utilities|Casting|CastToFortPlayerPawnAthena _aspawn)
              (:then
                (if (Vehicle|IsinVehicle _asfort_player_pawn_athena)
                  (bind _outhit (Collision|LineTraceByChannel (Math|Vector|SelectVector _location (Transformation|GetActorLocation) (Variables|GetbFromSweep)) (Transformation|GetActorLocation _asfort_player_pawn_athena) "TraceTypeQuery8" true (Utilities|Array|MakeArray _asfort_player_pawn_athena) (select (and (Variables|Default|GetbDebug) (Networking|HasAuthority)) "ForDuration" "None") true "(R=1.000000,G=0.000000,B=0.000000,A=1.000000)" "(R=0.000000,G=1.000000,B=0.000000,A=1.000000)" 1e+01))
                  (LocalVariables|SetOutValidHit (not (and (Collision|BreakHitResult _outhit) (== (.HitActor (Collision|BreakHitResult _outhit)) (Vehicle|GetVehicle _asfort_player_pawn_athena))))) ; Blocked by vehicle?
                  (Development|PrintString (Utilities|String|Append "Overlap Vehicle Check [" (Fortnite|GetPlayerNameSafe _asfort_player_pawn_athena) "] - " (Utilities|String|SelectString "Hit Vehicle " "Valid Overlap" (and (Collision|BreakHitResult _outhit) (== (.HitActor (Collision|BreakHitResult _outhit)) (Vehicle|GetVehicle _asfort_player_pawn_athena))))) (Variables|Default|GetbDebug) (Variables|Default|GetbDebug))
                  (else
                    _)))
              (:CastFailed
                _))
            (LocalVariables|SetOutValidHit true) ; Not a player in a vehicle - Return true
            (Development|PrintString (Utilities|String|Append "Valid Overlap [ " (Fortnite|GetPlayerNameSafe _asfort_player_pawn_athena) "]") (Variables|Default|GetbDebug) (Variables|Default|GetbDebug))
            (else
              _)))
        (:CastFailed
          _)) ; We only want to consider overlaps with pawns as hits
      (LocalVariables|SetOutValidHit false) ; Not a pawn or hit the instigator - Return false
    )
    (:then_1
      ;[1 0.6969 0.3949] Return
      (return (LocalVariables|GetOutValidHit)))))

(fn TrytoBreak
  (bind _stopnormal (Variables|FortLiquidGunProjectile|GetStopNormal))
  (bind _returnvalue (Utilities|FetchLiquidCenterPoint))
  (bind _output (Utilities|FetchLiquidStartPoint))
  (bind _returnvalue_2 (- _output _returnvalue))
  (bind _returnvalue_3 (Math|Vector|ProjectVectorontoPlane _returnvalue_2 _stopnormal))
  (bind _sf_anglebreaking_angleaggregationoffset (Variables|BreakageVariables|GetSF_AngleBreaking_AngleAggregationOffset))
  (bind _returnvalue_4 (ScalableFloat|GetValueatLevel _sf_anglebreaking_angleaggregationoffset))
  (bind _sf_anglebreaking_maxaggregateincreasemultiplier (Variables|BreakageVariables|GetSF_AngleBreaking_MaxAggregateIncreaseMultiplier))
  (bind _returnvalue_5 (ScalableFloat|GetValueatLevel _sf_anglebreaking_maxaggregateincreasemultiplier))
  (bind _sf_anglebreaking_upperlengthforminmultiplier (Variables|BreakageVariables|GetSF_AngleBreaking_UpperLengthForMinMultiplier))
  (bind _returnvalue_6 (ScalableFloat|GetValueatLevel _sf_anglebreaking_upperlengthforminmultiplier))
  (bind _sf_anglebreaking_minaggregateincreasemultiplier (Variables|BreakageVariables|GetSF_AngleBreaking_MinAggregateIncreaseMultiplier))
  (bind _returnvalue_7 (ScalableFloat|GetValueatLevel _sf_anglebreaking_minaggregateincreasemultiplier))
  (bind _sf_anglebreaking_prjtraveltimeformaxmultiplier (Variables|BreakageVariables|GetSF_AngleBreaking_PrjTravelTimeForMaxMultiplier))
  (bind _returnvalue_8 (ScalableFloat|GetValueatLevel _sf_anglebreaking_prjtraveltimeformaxmultiplier))
  (bind _stopnormal_2 (Variables|FortLiquidGunProjectile|GetStopNormal))
  (bind _output_2 (Utilities|FetchLiquidEndPoint))
  (bind _returnvalue_9 (- _returnvalue _output_2))
  (bind _returnvalue_10 (Math|Vector|ProjectVectorontoPlane _returnvalue_9 _stopnormal_2))
  (bind _angle__degrees (Class|AbilityFunctions|GetAngleBetweenTwoVectors _returnvalue_3 _returnvalue_10))
  (bind _returnvalue_11 (Math|Float|Square _angle__degrees))
  (bind _returnvalue_12 (Actor|GetGameTimeSinceCreation))
  (bind _output_3 (Utilities|NormalizeFloat _returnvalue_12 0 _returnvalue_8 true))
  (bind _lengthbetweenstartandend (LocalVariables|GetLengthBetweenStartAndEnd))
  (bind _output_4 (Utilities|RemapFloat _lengthbetweenstartandend (Math|Vector2D|MakeVector2D 0.0 _returnvalue_6) (Math|Vector2D|MakeVector2D _returnvalue_5 _returnvalue_7) true))
  (bind _returnvalue_13 (* _output_4 _output_3))
  (bind _returnvalue_14 (* _returnvalue_11 _returnvalue_13))
  (bind _returnvalue_15 (- _returnvalue_14 _returnvalue_4))
  (Utilities|IsValid (Variables|Default|GetCachedBPRibbonManager)
    (:Is%20Valid
      (if (or (CallFunction|BrokefromDistance) (Variables|Default|GetBlockAngleDetach))
        (return false)
        (elif (and (Utilities|IsValid (Variables|Default|GetPreviousPrj) :!class /Script/Engine.KismetSystemLibrary) (Utilities|IsValid (Variables|Default|GetNextPrj) :!class /Script/Engine.KismetSystemLibrary))
          (Networking|SwitchHasAuthority
            (:Authority
              (if (CallFunction|BrokefromAngle)
                (return true)
                (elif (Variables|Default|GetHasStopped)
                  (LocalVariables|SetLengthBetweenStartAndEnd (Math|Vector|VectorLength (- _output _output_2)))
                  (bind _asb_prj_ranged_nitroglycerine_gun_bb (Utilities|Casting|CastToB_Prj_Ranged_NitroglycerineGun_BB (Variables|Default|GetPreviousPrj)))
                  (if (and (> (+ _returnvalue_15 (Variables|BreakageVariables|GetAggregateAngle _asb_prj_ranged_nitroglycerine_gun_bb)) (ScalableFloat|GetValueatLevel (Variables|BreakageVariables|GetSF_AngleBreaking_ThresholdForBreak))) (> (LocalVariables|GetLengthBetweenStartAndEnd) (ScalableFloat|GetValueatLevel (Variables|BreakageVariables|GetSF_AngleBreaking_MinDistanceForAngleBreakage))))
                    (CallFunction|SetDetachmentReason "ByAngle") ; Mark that we should detach by angle
                    (Variables|BreakageVariables|SetAggregateAngle 0.0)
                    (return true)
                    (else
                      (Variables|BreakageVariables|SetAggregateAngle (Math|Float|Max(Float) (+ _returnvalue_15 (Variables|BreakageVariables|GetAggregateAngle _asb_prj_ranged_nitroglycerine_gun_bb)) -1e+02))
                      (return false)))
                  (else
                    (return false)))))
            (:Remote
              (return (CallFunction|BrokefromAngle))))
          (else
            (return false)))))) ; Valid check for the manager which is used in macros
)

(fn ConstructionScript)
