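# Masked, two-sided Niagara sprite material for Goo Gun liquid bubbles, faking a lit sphere on a flat sprite.
# A sphere normal map drives fake sun and view highlights, added to rim-gradient emissive colours that, like base colour, blend by particle LerpColor.
# Circular masks minus radially mapped cellular noise, scaled by particle Erode, build the dissolving opacity.
# Further dynamic parameters feed a rotated, cosine-rippled water-line gradient with fill and flip controls.
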Material /ForbiddenFruitWeaponsGameplay/Gameplay/Weapons/GooGun/FX/M_LiquidGun_Sprite_Bubble_BB
graph {
 properties {bUsedWithNiagaraSprites=true, blendMode=BLEND_Masked, pixelDepthOffsetMode=PDOM_Legacy, twoSided=true, usedShadingModels=Default Lit}
 nodes [
  Constant2Vector() {g=0.6, r=0.5}
  n1 = DynamicParameter() {defaultValue=<rgba 1 0 1 1>, paramNames [LerpColor,Erode,Param3,Param4]}
  LinearInterpolate(
   A=VectorParameter() {parameterName=BaseColorA, defaultValue=<rgba 0.0715 0.437 0.8573 1>}.RGB
   B=VectorParameter() {parameterName=BaseColorB, defaultValue=<rgba 0.0968 0.3904 0.6115 1>}.RGB
   Alpha=@n1.LerpColor
  )
  MaterialFunctionCall() {materialFunction=&DebugTimeSine}
  n2 = TextureSample(UVs=MaterialFunctionCall(Texture Scale=Constant() {r=1.3}) {materialFunction=&ScaleUVsByCenter}.UVs) {samplerType=SAMPLERTYPE_Normal, texture=/Game/Global/Textures/Default/Eyes/Sphere_EYE_NORMALS}
  MaterialFunctionCall(Normal=@n2.RGB, Flatness=Constant() {r=1.3}) {materialFunction=&FlattenNormal}
  n3 = ReflectionVectorWS(CustomWorldNormal=@n2.RGB) {bNormalizeCustomWorldNormal=true}
  n4 = DotProduct(A=CameraVectorWS(), B=@n3)
  n5 = CollectionParameter() {
   collection=&FortniteMaterialParameters
   parameterName=SunAndMoonModelDirectionalVector
  }
  n6 = DotProduct(A=ComponentMask(None=@n5) {b=true, g=true, r=true}, B=@n3)
  n7 = ScalarParameter() {defaultValue=30, parameterName=FakeSpecularPower}
  n8 = Add(A=Saturate(None=Multiply(A=@n4) {constB=0.2}), B=Saturate(None=Power(Base=@n6, Exp=@n7) {constExponent=20}))
  n9 = ScalarParameter() {defaultValue=0.5, parameterName=FakeSpecularMul}
  n10 = SphereMask(A=TextureCoordinate(), B=Constant() {r=0.5}, Radius=Constant() {r=0.5}, Hardness=Constant() {r=0.9})
  n11 = SphereMask(A=TextureCoordinate(), B=Constant() {r=0.5}, Radius=Constant() {r=0.46}, Hardness=Constant())
  n12 = Power(Base=Saturate(None=Subtract(A=@n10, B=@n11)), Exp=Constant() {r=8})
  n13 = LinearInterpolate(
   A=VectorParameter() {parameterName=EmissiveLow_A, defaultValue=<rgba 0.0045 0.3599 0.849 1>}.RGB
   B=VectorParameter() {parameterName=EmissiveHigh_A, defaultValue=<rgba 0.051 0.2433 1 1>}.RGB
   Alpha=@n12
  )
  n14 = LinearInterpolate(
   A=VectorParameter() {parameterName=EmissiveLow_B, defaultValue=<rgba 0.0475 0.0452 0.125 0>}.RGB
   B=VectorParameter() {parameterName=EmissiveHigh_B, defaultValue=<rgba 0.0234 0.0107 0.0729 0>}.RGB
   Alpha=@n12
  )
  MaterialFunctionCall(Emissive=Add(A=Multiply(A=@n8, B=@n9), B=LinearInterpolate(A=@n13, B=@n14, Alpha=@n1.LerpColor))) {materialFunction=&MF_FX_EmissiveAdjustments}
  n15 = SphereMask(A=TextureCoordinate(), B=Constant() {r=0.5}, Radius=Constant() {r=0.46}, Hardness=Constant() {r=1})
  n16 = Constant() {r=0.5}
  n17 = LinearInterpolate(Alpha=Subtract(A=NamedRerouteUsage(), B=ScalarParameter() {defaultValue=0.1, parameterName=Param_2})) {constA=-0.3, constB=0.6}
  Multiply(A=@n15, B=SphereMask(A=TextureCoordinate(), B=AppendVector(A=@n16, B=@n17), Radius=Constant() {r=0.8}, Hardness=Constant() {r=1}))
  n18 = NamedRerouteUsage()
  n19 = Multiply(
   A=MaterialFunctionCall() {materialFunction=&VectorToRadialValue}.Radial Coordinates
   B=Constant2Vector() {g=0.15, r=0.6}
  )
  n20 = Power(Base=OneMinus(None=TextureSample(UVs=@n19) {samplerType=SAMPLERTYPE_Masks, texture=/Game/Global/Textures/Default/Noise/T_CellularNoise}.R))
  n21 = LinearInterpolate(Alpha=DynamicParameter() {defaultValue=<rgba 1 0 1 1>, paramNames [LerpColor,Erode,Param3,Param4]}.Erode) {constB=2}
  NamedRerouteDeclaration(None=Subtract(A=Multiply(A=@n18, B=@n15), B=Multiply(A=@n20, B=@n21))) {name=OpacityMask, nodeColor=<rgba 0.0312 1 0 1>}
  n22 = DynamicParameter() {parameterIndex=3, paramNames [UpsideDown,Time,WaveAdd,Param4]}
  NamedRerouteDeclaration(None=Multiply(A=Add(A=@n22.UpsideDown)) {constB=0.5}) {name=WaterUpsideDown, nodeColor=<rgba 0 0.1797 1 1>}
  n23 = WorldPosition()
  n24 = NamedRerouteDeclaration(None=ObjectPositionWS()) {name=Position, nodeColor=<rgba 0.4297 0 1 1>}
  n25 = Multiply(A=NamedRerouteUsage(), B=Constant3Vector() {constant=<rgba -1 -1 1 0>})
  n26 = Divide(
   A=MaterialFunctionCall!(
    VectorToRotate=Subtract(A=@n23.XYZ, B=@n24)
    LookAtVector=Normalize(VectorInput=@n25)
    RestingVector=Constant3Vector() {constant=<rgba 0 0 1 0>}
   ) {materialFunction=/Engine/Functions/Engine_MaterialFunctions01/Vectors/RotateVector}.Result
   B=NamedRerouteUsage()
  ) {constB=1}
  n27 = Subtract(A=Multiply(B=ComponentMask(None=@n26) {b=true}) {constA=-1}, B=LinearInterpolate(Alpha=NamedRerouteUsage()) {constA=0.5, constB=-0.5})
  n28 = Power(Base=MaterialFunctionCall() {materialFunction=/Engine/Functions/Engine_MaterialFunctions02/Utility/VectorLength}.V2 Length)
  n29 = Multiply(A=Cosine(None=Multiply(A=@n28) {constB=0.5}), B=Multiply(A=NamedRerouteUsage()) {constB=-0.5}) {constB=0.15}
  n30 = OneMinus(None=Saturate(None=Add(A=@n27, B=@n29) {constB=0}))
  n31 = ScalarParameter() {defaultValue=0.1, parameterName=Param_1}
  NamedRerouteDeclaration(None=Power(Base=Subtract(A=@n30, B=@n31), Exp=ScalarParameter() {defaultValue=10})) {name=Water Line Gradient, nodeColor=<rgba 1 0 0.1172 1>}
  n32 = DynamicParameter() {parameterIndex=2}
  NamedRerouteDeclaration(None=Normalize(VectorInput=MaterialFunctionCall(X=@n32.Param1, Y=@n32.Param1, Z=@n32.Param1) {materialFunction=&MakeFloat3}.Result)) {name=WaterPlaneZ, nodeColor=<rgba 1 0 0.1875 1>}
  n33 = DynamicParameter() {parameterIndex=1}
  NamedRerouteDeclaration(None=@n33.Param4) {name=Fill, nodeColor=<rgba 0 1 0.0859 1>}
  NamedRerouteDeclaration(None=MaterialFunctionCall(X=@n33.Param1, Y=@n33.Param1, Z=@n33.Param1) {materialFunction=&MakeFloat3}.Result) {name=Dimensions, nodeColor=<rgba 0 0.0994 1 1>}
  NamedRerouteUsage()
  ScalarParameter() {defaultValue=0.1, parameterName=Roughness}
  ScalarParameter() {defaultValue=0.5, parameterName=Metallic}
  ScalarParameter() {defaultValue=1, parameterName=Specular}
 ]
}
