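# Lit goo gun liquid tube material for Niagara ribbons, with particle dynamic parameters driving color change, hit state and boiling.
# Swirl and noise masks blend blue and red-purple palettes with a side color, then particle color and axis-aligned fresnel shape emissive.
# The Hitted parameter switches masks and layered bubble normals from UV to animated world-aligned sampling.
# World-aligned noise displaces vertices along normals for boiling, and temporal dither takes a depth-offset parameter.
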Material /ForbiddenFruitWeaponsGameplay/Gameplay/Weapons/GooGun/FX/M_Liquid_Tube_V4_BB
graph {
 properties {bUsedWithNiagaraRibbons=true, pixelDepthOffsetMode=PDOM_Legacy, translucencyLightingMode=TLM_Surface, usedShadingModels=Default Lit}
 nodes [
  n1 = VectorParameter() {parameterName=MainColorA, defaultValue=<rgba 0 0.1448 0.3 1>}
  n2 = Multiply(A=@n1.RGB, B=VectorParameter() {parameterName=MainColorB, defaultValue=<rgba 0 0.3509 0.5052 1>}.RGB) {constB=0.4}
  n3 = TextureCoordinate()
  n4 = TextureObject() {samplerType=SAMPLERTYPE_LinearGrayscale, texture=/Game/Global/Textures/Default/Flowmaps/T_StormSwirlMask}
  n5 = Constant3Vector() {constant=<rgba 200 150 150 0>}
  n6 = Constant() {r=2.2}
  n7 = Multiply(A=Time(), B=Constant() {r=10}) {constB=50}
  n8 = NamedRerouteUsage()
  n9 = DynamicParameter() {parameterIndex=3, paramNames [Hitted,Param2,Param3,Param4]}
  n10 = LinearInterpolate(
   A=TextureSample(UVs=Multiply(A=@n3) {constB=0.5}, Tex=@n4).RGB
   B=MaterialFunctionCall(TextureObject=@n4, TextureSize=Multiply(A=@n5, B=@n6), WorldPosition=Add(A=@n7, B=@n8)) {materialFunction=/Game/Weapons/WeaponSkins/Wraps/MaterialFunctions/WorldAlignedTexture}.XYZ Texture
   Alpha=@n9.Hitted
  )
  n11 = Power(Base=@n10) {constExponent=0.4}
  n12 = VectorParameter() {parameterName=MainColorC, defaultValue=<rgba 0.5625 0.029 0.0234 0>}
  n13 = Multiply(A=@n12.RGB, B=VectorParameter() {parameterName=MainColorD, defaultValue=<rgba 0.0342 0.0147 0.2708 0>}.RGB)
  n14 = Power(Base=@n10) {constExponent=0.5}
  n15 = DynamicParameter() {parameterIndex=1, defaultValue=<rgba 0 1 1 1>, paramNames [ChangeColorToPurple, PixelDepthOffset, Param3, Param4]}
  n16 = TextureObject() {samplerType=SAMPLERTYPE_Masks, texture=/Game/Weapons/WeaponSkins/Wraps/Textures/Mash/Mash_Noise}
  n17 = Constant3Vector() {constant=<rgba 300 300 300 0>}
  n18 = Constant() {r=3}
  n19 = Multiply(A=Time(), B=Constant() {r=20}) {constB=50}
  n20 = NamedRerouteUsage()
  n21 = LinearInterpolate(
   A=TextureSample(Tex=@n16).RGB
   B=MaterialFunctionCall(TextureObject=@n16, TextureSize=Multiply(A=@n17, B=@n18), WorldPosition=Add(A=@n19, B=@n20)) {materialFunction=&WorldAlignedTexture}.XYZ Texture
   Alpha=@n9.Hitted
  )
  n22 = Constant()
  n23 = LinearInterpolate(
   A=LinearInterpolate(A=LinearInterpolate(A=@n2, B=@n1.RGB, Alpha=@n11), B=LinearInterpolate(A=@n13, B=@n12.RGB, Alpha=@n14), Alpha=@n15.ChangeColorToPurple)
   B=VectorParameter() {parameterName=SideColor, defaultValue=<rgba 0.0583 0 0.1 1>}.RGB
   Alpha=Multiply(A=MaterialFunctionCall(In=@n21, Contrast=@n22) {materialFunction=&CheapContrast}.Result) {constB=0.7}
  )
  n24 = ParticleColor()
  n25 = ScalarParameter() {defaultValue=5, parameterName=Param_7}
  n26 = ScalarParameter() {defaultValue=0.2, parameterName=Param_9}
  n27 = Power(
   Base=MaterialFunctionCall!(Axis=Constant3Vector() {constant=<rgba 1 0 0 0>}, Exponent, Normal) {materialFunction=/Game/Effects/Fort_Effects/Materials/_Functions/MF_AxisAlignedFrensnel}.Fresnel
   Exp=ScalarParameter() {defaultValue=8, parameterName=Param_8}
  )
  n28 = Multiply(A=@n23, B=Multiply(A=Multiply(A=@n24.RGB, B=@n25), B=LinearInterpolate(A=@n26, Alpha=@n27) {constA=0.2}))
  MaterialFunctionCall(Emissive=@n28) {materialFunction=&MF_FX_EmissiveAdjustments}
  n29 = DynamicParameter() {parameterIndex=1, defaultValue=<rgba 0 1 1 1>, paramNames [ChangeColorToPurple, PixelDepthOffset, Param3, Param4]}
  MaterialFunctionCall(
   Alpha Threshold=ScalarParameter() {defaultValue=0.5, parameterName=Param_5}
   Random=LinearInterpolate(B=ScalarParameter() {defaultValue=350, parameterName=Param_6}, Alpha=@n29.PixelDepthOffset) {constAlpha=20}
  ) {materialFunction=&DitherTemporalAA_Fortnite}
  NamedRerouteDeclaration(None=@n28) {name=Emissive, nodeColor=<rgba 0 0.9062 1 1>}
  NamedRerouteDeclaration(None=@n23) {name=BaseColor, nodeColor=<rgba 0.4766 0 1 1>}
  n30 = Multiply(A=Time(), B=Constant() {r=100}) {constB=50}
  n31 = NamedRerouteUsage()
  n32 = MaterialFunctionCall(Float3=MaterialFunctionCall(
    TextureObject=TextureObject() {samplerType=SAMPLERTYPE_LinearColor, texture=&HighContrastNoise}
    TextureSize=Constant() {r=100}
    WorldPosition=Add(A=@n30, B=@n31)
   ) {desc=Boiling, materialFunction=&WorldAlignedTexture}.XYZ Texture) {materialFunction=&BreakOutFloat3Components}
  n33 = Multiply(
   A=VertexNormalWS()
   B=MaterialFunctionCall(R=LinearInterpolate(Alpha=@n32.R) {constA=-1}, G=LinearInterpolate(Alpha=@n32.G) {constA=-1}, B=LinearInterpolate(Alpha=@n32.B) {constA=-1}) {materialFunction=&AppendMany}.RGB
  )
  n34 = Multiply(
   A=DynamicParameter() {defaultValue=<rgba 1 1 0 1>, paramNames [WPO_Mulit,WPO_ZPlane,BoilingWPO,Param4]}.BoilingWPO
   B=ScalarParameter() {defaultValue=2}
  ) {constB=200}
  NamedRerouteDeclaration(None=Add(B=Multiply(A=@n33, B=@n34))) {name=WPO, nodeColor=<rgba 0 1 0.1562 1>}
  NamedRerouteDeclaration(None=Subtract(A=WorldPosition().XYZ, B=ActorPositionWS())) {name=Position, nodeColor=<rgba 1 0.037 0 1>}
  n35 = TextureCoordinate()
  n36 = TextureObject() {samplerType=SAMPLERTYPE_Normal, texture=/Game/Global/Textures/Default/Noise/T_Bubble_B_N}
  n37 = Add(A=Multiply(A=Time(), B=Constant() {r=10}) {constB=50}, B=NamedRerouteUsage())
  n38 = LinearInterpolate(
   A=TextureSample(UVs=Multiply(A=@n35) {constB=0.25}, Tex=@n36).RGB
   B=MaterialFunctionCall(TextureObject=@n36, TextureSize=Constant() {r=800}, WorldPosition=@n37) {materialFunction=&WorldAlignedTexture}.XYZ Texture
   Alpha=@n9.Hitted
  )
  n39 = LinearInterpolate(
   A=TextureSample(Tex=@n36).RGB
   B=MaterialFunctionCall(TextureObject=@n36, TextureSize=Constant() {r=200}, WorldPosition=@n37) {materialFunction=&WorldAlignedTexture}.XYZ Texture
   Alpha=@n9.Hitted
  )
  NamedRerouteDeclaration(None=MaterialFunctionCall(
    Normal=MaterialFunctionCall(BaseNormal=@n38, AdditionalNormal=@n39) {materialFunction=&BlendAngleCorrectedNormals}.Result
    Flatness=ScalarParameter() {defaultValue=1.2, parameterName=NormalIntensity}
   ) {materialFunction=&FlattenNormal}.Result) {name=PixelNormal, nodeColor=<rgba 0.0434 0.1768 1 1>}
  NamedRerouteUsage()
  NamedRerouteUsage()
  NamedRerouteUsage()
  NamedRerouteUsage()
  ScalarParameter() {defaultValue=0.15, parameterName=Roughness}
  ScalarParameter() {defaultValue=0.5, parameterName=Metallic}
  ScalarParameter() {defaultValue=1, parameterName=Specular}
 ]
}
