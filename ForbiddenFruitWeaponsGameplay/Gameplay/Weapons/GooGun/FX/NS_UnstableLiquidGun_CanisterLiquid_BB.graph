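# Niagara system for the Unstable Liquid Gun (Goo Gun) canister, rendering sloshing liquid with bubbles and suck-out effects.
# A spring pendulum follows owner motion to derive water plane and agitation, feeds sloshing audio, and is read by liquid mesh and bubble emitters.
# PistonProgress (a waveform in Debug) squashes the mesh and clips bubbles, whose spawning scales with system velocity.
# PlaySuckOutFX gates GPU bubble and CPU splash sprite spawning.
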NiagaraSystem /ForbiddenFruitWeaponsGameplay/Gameplay/Weapons/GooGun/FX/NS_UnstableLiquidGun_CanisterLiquid_BB
dependencies {
 usedDataInterfaces [
  &NiagaraDataInterfaceMeshRendererInfo
  &NiagaraDataInterfaceParticleRead
  &NiagaraDataInterfaceCurve
 ]
 usedDynamicInputs [&Divide_Float]
 usedModules [
  &self:SystemSpawnScript.SystemScriptSource.SystemScriptGraph.NiagaraNodeAssignment_0.SetVariables_#1
  &SystemState
  &self:SystemSpawnScript.SystemScriptSource.SystemScriptGraph.NiagaraNodeAssignment_1.SetVariables_#2
  &self:SystemSpawnScript.SystemScriptSource.SystemScriptGraph.NiagaraNodeAssignment_2.SetVariables_#3
 ]
}
emitters {
 Bubble001 {
  emitterUpdateScript [
   &EmitterState(Life Cycle Mode=::NewEnumerator1, UseLoopDelay=false)
   &InterpolateOverTime(Rate Of Change=2, Target Float Value:NiagaraFloat=$System.SysVelocity)
   &SpawnRate(
    SpawnRate=&Lerp_Float(A=20, B=1000, Alpha=$Emitter.InterpolateOverTime.InterpolateOverTime_Float.Moving Average)
    Use Spawn Probability=true
   )
  ]
  particleSpawnScript [
   &InitializeParticle(
    Sprite Rotation Mode=::NewEnumerator2
    Sprite Size Mode=::NewEnumerator1
    Uniform Sprite Size Max=4
    Uniform Sprite Size Min=3
    UsePositionOffset=false
    Write Material Random=true
   )
   &ShapeLocation(
    Cylinder Height=12
    Cylinder Radius=5
    MaskSpawnGroup=false
    Override Seed=false
    Rotation Axis=<0 1 0>
    Rotation Mode=::NewEnumerator1
    Shape Primitive=::NewEnumerator1
    Surface Only=false
    UseOwnerScale=true
   )
   &AddVelocity(
    Rotation Coordinate Space=::World
    Velocity=&MakeVector(X=0, Y=0, Z=&RandomRangeFloat(Minimum=1, Maximum=2))
   )
   &CameraOffset(Camera Offset Amount=-2)
  ]
  particleUpdateScript [
   &ParticleState()
   &self:Initialize_Liquid_06(Dimensions:Vector3f=<10 10 10>, Liquid Fill Amount:NiagaraFloat=$System.FillAmount) {module=Initialize_Liquid_06}
   &InterpolateOverTime(Rate Of Change=0.02, Target Float Value:NiagaraFloat=$System.SysVelocity) {module=InterpolateOverTime001}
   &RegionMask(
    Offset:Vector3f=<0 0 5>
    Offset Space:ENiagaraCoordinateSpace=::World
    Region Dimensions:Vector3f=&MakeVector(X=20, Y=20, Z=20)
    Region Falloff=1
    Region Type:ENiagaraDebugDrawShape=::NewEnumerator1
    UseOffset=true
    UseRotation=false
   )
   &ScaleSpriteSize(
    Uniform Curve Scale=&Multiply_Float(
     A=$StackContext.RegionMask.Float
     B=&Lerp_Float(A=1, B=1.5, Alpha=$Particles.InterpolateOverTime001.InterpolateOverTime_Float.Moving Average)
    )
    Uniform Curve Sprite Scale=DataInterface() {propertyValues {Curve {keys [
        {arriveTangent=4.1785, interpMode=RCIM_Cubic, leaveTangent=4.1785, tangentMode=RCTM_User, time=0, value=0}
        {arriveTangent=0.9995, interpMode=RCIM_Cubic, leaveTangent=0.9995, tangentMode=RCTM_User, time=0.2223, value=0.6825}
        {arriveTangent=1.6153, interpMode=RCIM_Cubic, leaveTangent=1.6153, time=0.88, value=1.0575}
        <at 0.96 1.8742>
        {arriveTangent=-0.0405, interpMode=RCIM_Cubic, leaveTangent=-0.0405, tangentMode=RCTM_User, time=1, value=0}
       ]}}}
   )
   &KillParticlesInVolume(
    Invert Volume=true
    Kill Shape=::NewEnumerator3
    Origin Offset:Vector3f=&TransformVector(Vector=<0 0 2>, Source Space=::World, Destination Space=::Local)
    Slab Axis:Vector3f=&TransformVector(Vector=<0 0 1>, Source Space=::World, Destination Space=::Local)
    Slab Width=4
   ) {module=KillParticlesInVolume001}
   &KillParticlesInVolume(
    Kill Shape=::NewEnumerator2
    Origin Offset:Vector3f=&MakeVector(
     X=0
     Y=&Lerp_Float(
      A=6
      B=-6
      Alpha=&MakeCustomFloatFromBool(
       Bool=$User.Debug
       True Float=&Waveform("Waveform [1]"=::NewEnumerator0, "[1] Frequency"=1, "[1] Amplitude Min/Max"=<0 1>)
       False Float=$User.PistonProgress
      )
     )
     Z=0
    )
    Plane Normal:Vector3f=<0 1 0>
   ) {module=KillParticlesInVolume002}
   &CurlNoiseForce(
    Noise Frequency=500
    Noise Strength=&Lerp_Float(A=5, B=50, Alpha=$Particles.InterpolateOverTime001.InterpolateOverTime_Float.Moving Average)
    Pan Noise=true
    Pan Noise Field=<0.5 0.5 0.5>
   )
   &Drag(Drag=2, Use Linear Drag=true, Use Rotational Drag=true)
   &SolveForcesAndVelocity(Clamp Velocity=false, Limit Acceleration=false)
   &GenerateDeathEvent(Use Event Probability=false)
   &DynamicMaterialParameters(
    Index 0 Param 1=&FloatFromCurve(
     FloatCurve=DataInterface() {propertyValues {Curve {keys [
         <at 0 0 RCIM_Cubic>
         {arriveTangent=0.8574, interpMode=RCIM_Cubic, leaveTangent=0.8574, time=0.6538, value=0.172}
         {arriveTangent=2.3917, interpMode=RCIM_Cubic, leaveTangent=2.3917, time=0.8801, value=0.7545}
         {arriveTangent=0.6955, interpMode=RCIM_Cubic, leaveTangent=0.6955, tangentMode=RCTM_User, time=1, value=1}
        ]}}}
     Scale Curve=&Lerp_Float(A=0.1, B=1, Alpha=$Particles.InterpolateOverTime001.InterpolateOverTime_Float.Moving Average)
    )
    Index 0 Param 2=&FloatFromCurve(
     FloatCurve=DataInterface() {propertyValues {Curve {keys [
         <at 0 0>
         {arriveTangent=1, interpMode=RCIM_Cubic, leaveTangent=1, time=0.85, value=0}
         {arriveTangent=3.1127, interpMode=RCIM_Cubic, leaveTangent=3.1127, tangentMode=RCTM_User, time=1, value=1}
        ]}}}
     Scale Curve=0.5
    )
    Module.Param0WriteEnabled=true
    Module.Param1WriteEnabled=true
    Module.Param2WriteEnabled=true
    Module.Param3WriteEnabled=true
   )
   &self:ScratchModule_04(New Particle Attribute Reader=DataInterface() {propertyValues {EmitterBinding {bindingMode=Other, emitterName=ChromePendulum}}}) {module=ScratchModule_04}
  ]
  rendererClasses [&NiagaraSpriteRendererProperties]
  renderers [{rendererClass=&NiagaraSpriteRendererProperties}]
 }
 ChromeLiquidMesh {
  emitterUpdateScript [
   &EmitterState(
    Inactive Response=::NewEnumerator1
    Life Cycle Mode=::NewEnumerator1
    Loop Behavior=::NewEnumerator1
    Loop Duration Mode=::NewEnumerator1
    UseLoopDelay=false
   )
   &SpawnBurst_Instantaneous(Use Spawn Probability=false, UseLoopCountLimit=false)
  ]
  particleSpawnScript [
   &InitializeParticle(UsePositionOffset=false, Write Material Random=true)
   &self:Initialize_Liquid_05(
    Liquid Fill Amount=0.85
    Liquid Mesh Renderer=DataInterface() {propertyValues {MeshRenderer=&self:ChromeLiquidMesh_4.NiagaraMeshRendererProperties_0}}
   ) {module=Initialize_Liquid_05}
   &CameraOffset(Camera Offset Amount=0.5)
  ]
  particleUpdateScript [
   &ParticleState(Kill Particles When Lifetime Has Elapsed=false)
   &ScaleMeshSize(Scale Factor=&MakeVector(
     X=0.98
     Y=&Lerp_Float(
      A=0
      B=0.7
      Alpha=&MakeCustomFloatFromBool(
       Bool=$User.Debug
       True Float=&Waveform("Waveform [1]"=::NewEnumerator0, "[1] Frequency"=1, "[1] Amplitude Min/Max"=<0 1>)
       False Float=$User.PistonProgress
      )
     )
     Z=0.98
    ))
   &self:ScratchModule_02(New Particle Attribute Reader=DataInterface() {propertyValues {EmitterBinding {bindingMode=Other, emitterName=ChromePendulum}}}) {module=ScratchModule_02}
   &SystemLocation(
    MaskSpawnGroup=false
    Offset=&MakeVector(
     X=0
     Y=&Add_Float(
      A=5
      B=&Lerp_Float(
       A=3
       B=-4.8
       Alpha=&MakeCustomFloatFromBool(
        Bool=$User.Debug
        True Float=&Waveform("Waveform [1]"=::NewEnumerator0, "[1] Frequency"=1, "[1] Amplitude Min/Max"=<0 1>)
        False Float=$User.PistonProgress
       )
      )
     )
     Z=0
    )
   )
   /SetVariables(Particles.DynamicMaterialParameter3:Vector4f=&MakeFloat4FromVectorAndFloat("Vector (RGB)"=<10 10 10>, "Float (Alpha)"=$System.FillAmount)) {isSetParameters=true}
  ]
  rendererClasses [&NiagaraMeshRendererProperties]
  renderers [{rendererClass=&NiagaraMeshRendererProperties, rendererIndex=0}]
 }
 ChromePendulum {
  emitterUpdateScript [
   &EmitterState(
    Inactive Response=::NewEnumerator1
    Life Cycle Mode=::NewEnumerator1
    Loop Behavior=::NewEnumerator1
    Loop Duration Mode=::NewEnumerator1
    UseLoopDelay=false
   )
   &SpawnBurst_Instantaneous(Use Spawn Probability=false, UseLoopCountLimit=false)
   /SetVariables(Emitter.Orientation:Quat4f=$Engine.Owner.Rotation) {isSetParameters=true}
   /SetVariables(Emitter.RotationVelocity:Vector3f=&Divide_Vector(
     A:Vector3f=&Subtract_Vector(
      A=&RotateVectorByQuaternion(Vector To Rotate=<0 0 1>, Quaternion=$Emitter.Orientation)
      B=&RotateVectorByQuaternion(Vector To Rotate=<0 0 1>, Quaternion=$Emitter.Previous.Orientation)
     )
     B:Vector3f=&VectorFromFloat(Value=$Engine.DeltaTime)
    )) {isSetParameters=true}
  ]
  particleSpawnScript [
   &InitializeParticle(Sprite Size Mode=::NewEnumerator3, UsePositionOffset=false, Write Material Random=true)
   /Niagara/Modules/Audio/PlayPersistentAudio(Audio Player:NiagaraDataInterfaceAudioPlayer=$User.SloshingAudio)
  ]
  particleUpdateScript [
   &ParticleState(Kill Particles When Lifetime Has Elapsed=false)
   /SetVariables(Particles.Orientation3:Quat4f=$Engine.Owner.Rotation) {isSetParameters=true}
   /SetVariables(Particles.Position_Origin:NiagaraPosition=$Engine.Owner.Position) {isSetParameters=true}
   &Drag(Drag=30, Use Linear Drag=true, Use Rotational Drag=false)
   &PendulumSetup(
    Module.Enable Spring Driver=true
    Pendulum Length=5
    Pendulum Pivot:NiagaraPosition=$Particles.Position_Origin
    "Spring Driven Constraint (Tightness)"=1.5
   )
   &InheritSourceMovement(
    Applied Position Scale=&Multiply_VectorByFloat(Vector=<1 1 1>, Float=1)
    ApplyForce=false
    ApplyPosition=true
    ApplyPositionFalloff=false
    Source Velocity:Vector3f=$Emitter.RotationVelocity
   )
   &InheritVelocity(
    Inherited Velocity Amount Scale=&Multiply_VectorByFloat(Vector=<1 1 1>, Float=1)
    Speed Limit=false
    UseSpeedThreshold=false
   )
   &SolveForcesAndVelocity(Clamp Velocity=false, Limit Acceleration=false, Speed Limit=200)
   &PendulumConstraint()
   &CalculateAccurateVelocity()
   &InterpolateOverTime(Target Float Value:NiagaraFloat=&FloatFromCurve(
     FloatCurve=DataInterface() {propertyValues {Curve {keys [
         {interpMode=RCIM_Cubic, tangentMode=RCTM_User, time=0, value=0}
         {arriveTangent=1.6368, interpMode=RCIM_Cubic, leaveTangent=1.6368, tangentMode=RCTM_User, time=0.3, value=0.35}
         {interpMode=RCIM_Cubic, tangentMode=RCTM_User, time=1, value=1}
        ]}}}
     CurveIndex=&Multiply_Float(A=&VectorLength(Vector=$Particles.Velocity), B=0.0025)
    ))
   &InterpolateOverTime(
    Rate Of Change=2
    Target Vector Value:Vector3f=&Subtract_Vector(
     A=Linked() {linkedVariable {name=Particles.Position_Origin, type=NiagaraPosition}}
     B=Linked() {linkedVariable {name=Particles.Position, type=NiagaraPosition}}
    )
    Variable Type=::NewEnumerator2
   ) {module=InterpolateOverTime002}
   /SetVariables(Particles.WaterPlaneUp:Vector3f=&NormalizeVector(
     Vector To Normalize=$Particles.InterpolateOverTime002.InterpolateOverTime_V3_.Moving Average
     Safe Normalize=true
     Safe Fallback Threshold=0
     Safe Fallback Vector=<0 0 0.001>
    )) {isSetParameters=true}
   &InterpolateOverTime(
    Rate Of Change=2
    Target Float Value:NiagaraFloat=&Add_Float(
     A=&DotProduct(Vector A=<0 0 -1>, Vector B=$Particles.WaterPlaneUp)
     B=&ClampFloat(Float=&DotProduct(
       Vector A=&NormalizeVector(Vector To Normalize=$Particles.Velocity, Safe Normalize=true, Safe Fallback Threshold=0, Safe Fallback Vector=<0 0 0.001>)
       Vector B=<0 0 -1>
      ))
    )
   ) {module=InterpolateOverTime001}
   /Niagara/Modules/Update/Utility/IncrementOverTime(Rate of Change:NiagaraFloat=&Lerp_Float(A=0.15, B=1.75, Alpha=$Particles.InterpolateOverTime.InterpolateOverTime_Float.Moving Average))
   /SetVariables(Particles.WaterPlaneZ:Vector4f=&MakeFloat4FromVectorAndFloat("Vector (RGB)"=$Particles.WaterPlaneUp, "Float (Alpha)"=$Particles.InterpolateOverTime.InterpolateOverTime_Float.Moving Average)) {isSetParameters=true}
   /SetVariables(Particles.WaterParams2:Vector4f=&MakeVector4(X=$Particles.InterpolateOverTime001.InterpolateOverTime_Float.Moving Average, Y=$Particles.IncrementOverTime.Counter, Z=0.1, W=0)) {isSetParameters=true}
   /Niagara/Modules/Audio/UpdatePersistentAudio(
    Audio Player:NiagaraDataInterfaceAudioPlayer=$User.SloshingAudio
    Float Value:NiagaraFloat=&RemapRange(
     Input Value To Remap=$Particles.InterpolateOverTime.InterpolateOverTime_Float.Moving Average
     Input Min=0
     Input Max=1
     Output Min=0
     Output Max=1
    )
    Parameter Name Index=1
    Update Type:ENiagara_AudioParamType=::NewEnumerator2
   )
   &UpdatePersistentAudio(
    Audio Player:NiagaraDataInterfaceAudioPlayer=$User.SloshingAudio
    Float Value:NiagaraFloat=$User.PistonProgress
    Update Type:ENiagara_AudioParamType=::NewEnumerator2
   ) {module=UpdatePersistentAudio001}
   &UpdatePersistentAudio(
    Audio Player:NiagaraDataInterfaceAudioPlayer=$User.SloshingAudio
    Volume:NiagaraFloat=&MakeCustomFloatFromBool(Bool=$User.WeaponVisible, True Float=1, False Float=0)
   ) {module=UpdatePersistentAudio002}
  ]
  rendererClasses [&NiagaraSpriteRendererProperties]
  renderers [{rendererClass=&NiagaraSpriteRendererProperties}]
 }
 SuckBubbles {
  simTarget=GPUComputeSim
  emitterUpdateScript [
   &EmitterState(Life Cycle Mode=::NewEnumerator1, Loop Duration=2, UseLoopDelay=false)
   &SpawnRate(Spawn Probability=$User.PlaySuckOutFX, SpawnRate=80, Use Spawn Probability=true)
  ]
  particleSpawnScript [
   &InitializeParticle(
    Lifetime Max=0.8
    Lifetime Min=0.1
    Lifetime Mode=::NewEnumerator1
    Sprite Rotation Mode=::NewEnumerator2
    Sprite Size Max=<2 3>
    Sprite Size Min=<1 2>
    Sprite Size Mode=::NewEnumerator2
    UsePositionOffset=false
    Write Material Random=true
   ) {module=InitializeParticle001}
   &SystemLocation(MaskSpawnGroup=false, Offset=<0 8 0>)
   &DynamicMaterialParameters(
    Index 0 Param 1=&RandomRangeFloat(Minimum=-0.5, Maximum=0.5)
    Module.Param0WriteEnabled=true
    Module.Param1WriteEnabled=true
    Module.Param2WriteEnabled=true
    Module.Param3WriteEnabled=true
   )
   &AddVelocity(
    Cone Angle=70
    Cone Axis=<0 -1 0>
    Module.UseConeFalloff=false
    Module.UseVelDistribution=false
    Override Seed=false
    Velocity Mode=::NewEnumerator2
    Velocity Speed=90
   )
  ]
  particleUpdateScript [
   &ParticleState()
   &ScaleSpriteSize(Uniform Curve Sprite Scale=DataInterface() {propertyValues {Curve {keys [
        {arriveTangent=9.0725, interpMode=RCIM_Cubic, leaveTangent=9.0725, tangentMode=RCTM_User, time=0, value=0}
        {arriveTangent=-2.5249, interpMode=RCIM_Cubic, leaveTangent=-2.5249, tangentMode=RCTM_User, time=0.1408, value=0.9847}
        {arriveTangent=-0.0405, interpMode=RCIM_Cubic, leaveTangent=-0.0405, tangentMode=RCTM_User, time=1, value=0}
       ]}}})
   &Drag(Drag=6, Use Linear Drag=true, Use Rotational Drag=true)
   &GravityForce(Gravity=<0 0 -30>)
   &CurlNoiseForce(Noise Frequency=100, Noise Strength=50, Pan Noise=false)
   &SolveForcesAndVelocity(Clamp Velocity=false, Limit Acceleration=false)
  ]
  rendererClasses [&NiagaraSpriteRendererProperties]
  renderers [{rendererClass=&NiagaraSpriteRendererProperties}]
 }
 SuckSpalshes {
  emitterUpdateScript [
   &EmitterState(Life Cycle Mode=::NewEnumerator1, Loop Duration=2, UseLoopDelay=false)
   &SpawnRate(Spawn Probability=$User.PlaySuckOutFX, SpawnRate=10, Use Spawn Probability=true)
  ]
  particleSpawnScript [
   /Niagara/Modules/Spawn/Initialization/InitializeParticle(
    Color=<rgba 2 2 2 1>
    Lifetime=&RandomRangeFloat(Minimum=0.5, Maximum=0.6)
    Mesh Scale=<10 10 10>
    Module.Write Color=true
    Module.Write Lifetime=true
    Module.Write Mass=false
    Module.Write Material Random=true
    Module.Write Position=true
    Module.Write Scale=true
    Module.Write SpriteRotation=true
    Module.Write SpriteSize=true
    Sprite Rotation=&RandomRangeFloat(Minimum=-180, Maximum=180)
    Sprite Size=&Multiply_Vector2D_ByFloat(Float=0.15, Vector2D=&RandomRangeVector2D(Minimum=<75 75>, Maximum=<150 150>))
   )
   /SetVariables(Particles.UVScale:Vector2f=&MakeVector2D(X=&UniformAOrBFloat(A=-1), Y=1)) {isSetParameters=true}
   &SystemLocation(MaskSpawnGroup=false, Offset=<0 8 0>)
   &DynamicMaterialParameters(
    Index 0 Param 1=&RandomRangeFloat(Minimum=0.1, Maximum=1)
    Index 0 Param 2=&RandomRangeFloat(Minimum=0.2, Maximum=0.5)
    Index 0 Param 3=&RandomRangeFloat(Minimum=50, Maximum=100)
    Index 0 Param 4=&RandomRangeFloat(Minimum=0, Maximum=10)
    Module.Param0WriteEnabled=true
    Module.Param1WriteEnabled=true
    Module.Param2WriteEnabled=true
    Module.Param3WriteEnabled=true
   ) {module=DynamicMaterialParameters001}
   &AddVelocity(
    Cone Axis=<0 -1 0>
    Module.UseConeFalloff=false
    Module.UseVelDistribution=false
    Override Seed=false
    Velocity Mode=::NewEnumerator2
    Velocity Speed=50
   )
   &Color(
    Color=<rgba 3 0.2244 0.03 1>
    Scale Alpha=&FloatFromCurve(FloatCurve=DataInterface() {propertyValues {Curve {keys [
         {interpMode=RCIM_Cubic, tangentMode=RCTM_Break, time=0, value=1}
         {arriveTangent=-3.5002, interpMode=RCIM_Cubic, tangentMode=RCTM_Break, time=1, value=0}
        ]}}})
   ) {module=Color001}
  ]
  particleUpdateScript [
   &ParticleState()
   &GravityForce(Gravity=<0 0 -30>)
   &Drag(Drag=6, Use Linear Drag=true, Use Rotational Drag=false)
   &SolveForcesAndVelocity(Clamp Velocity=false, Limit Acceleration=false)
   &ScaleSpriteSize(Scale Factor=&Vector2DFromFloat(Value=&FloatFromCurve(
      FloatCurve=DataInterface() {propertyValues {Curve {keys [
          {arriveTangent=5.9795, interpMode=RCIM_Cubic, leaveTangent=5.9795, tangentMode=RCTM_User, time=0, value=0.2}
          {arriveTangent=-0.5, interpMode=RCIM_Cubic, leaveTangent=-0.5, time=0.7, value=3}
          {arriveTangent=-24.1897, interpMode=RCIM_Cubic, leaveTangent=-24.1897, tangentMode=RCTM_User, time=1, value=0}
         ]}}}
      Scale Curve=0.4
     ))) {module=ScaleSpriteSize001}
   &ScaleColor(
    Scale Alpha=&FloatFromCurve(FloatCurve=DataInterface() {propertyValues {Curve {keys [
         {arriveTangent=-2.5249, interpMode=RCIM_Cubic, time=0.1, value=1}
         {arriveTangent=-0.9941, interpMode=RCIM_Cubic, leaveTangent=-0.9941, tangentMode=RCTM_User, time=0.3843, value=0.3455}
         {arriveTangent=-1.0623, interpMode=RCIM_Cubic, leaveTangent=-1.0623, tangentMode=RCTM_User, time=1, value=0}
        ]}}})
    Scale RGB=&MakeVectorFromLinearColorRGB(Color=<rgba 1 1 1 1>)
    ScaleA=true
    ScaleRGB=false
   )
   &SpriteRotationRate(Rotation Rate=10)
   &DynamicMaterialParameters(
    Index 1 Param 1=&FloatFromCurve(FloatCurve=DataInterface() {propertyValues {Curve {keys [
         {arriveTangent=4.2071, interpMode=RCIM_Cubic, time=0, value=-2}
         {arriveTangent=1.1564, interpMode=RCIM_Cubic, leaveTangent=1.1564, tangentMode=RCTM_User, time=1, value=1}
        ]}}})
    Module.Param0WriteEnabled1=true
    Module.Param1WriteEnabled1=false
    Module.Param2WriteEnabled1=false
    Module.Param3WriteEnabled1=false
    Write Parameter Index 0=false
    Write Parameter Index 1=true
   )
  ]
  rendererClasses [&NiagaraSpriteRendererProperties]
  renderers [{rendererClass=&NiagaraSpriteRendererProperties}]
 }
}
summary {
 systemName=NS_UnstableLiquidGun_CanisterLiquid_BB
 emitters [
  {emitterName=ChromeLiquidMesh, rendererClasses [&NiagaraMeshRendererProperties]}
  {emitterName=ChromePendulum, rendererClasses [&NiagaraSpriteRendererProperties]}
  {emitterName=Bubble001, rendererClasses [&NiagaraSpriteRendererProperties]}
  {emitterName=SuckBubbles, simTarget=GPUComputeSim, rendererClasses [&NiagaraSpriteRendererProperties]}
  {emitterName=SuckSpalshes, rendererClasses [&NiagaraSpriteRendererProperties]}
 ]
 userVariables [
  User.bIsFrontEnd {type=NiagaraBool}
  User.bIsFrontEndPreview {type=NiagaraBool}
  User.Debug {type=NiagaraBool}
  User.FillAmount {type=NiagaraFloat, defaultValue=1, description=""}
  User.PistonProgress {type=NiagaraFloat, defaultValue=1, description=""}
  User.PlaySuckOutFX {type=NiagaraFloat, defaultValue=0, description=""}
  User.SloshingAudio {
   type=NiagaraDataInterfaceAudioPlayer
   description=""
   defaultValue {
    struct=NiagaraExt_VariableValue_DataInterface
    value {
     dataInterface=&self:NiagaraDataInterfaceAudioPlayer_0
     dataInterfaceClass=&NiagaraDataInterfaceAudioPlayer
    }
   }
  }
  User.WeaponVisible {type=NiagaraBool}
 ]
}
